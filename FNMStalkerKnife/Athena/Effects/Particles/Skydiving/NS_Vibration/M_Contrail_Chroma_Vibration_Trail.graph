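# Translucent, two-sided unlit Niagara material for a chromatic, flickering skydiving contrail trail.
# Scrolling flow and rainbow textures get flicker-driven chroma offsets and noise distortion, then are tinted by inside color, optional curve, particle color and time of day.
# Opacity combines base and scrolling smoke masks with camera and depth fades, plus optional bottom and Fresnel fades.
# Dynamic parameters vary tiling, timing and scroll; refraction and normals drop out on cheaper quality and shading paths.
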Material /FNMStalkerKnife/Athena/Effects/Particles/Skydiving/NS_Vibration/M_Contrail_Chroma_Vibration_Trail
graph {
 properties {
  bUseTranslucencyVertexFog=false
  bUsedWithNiagaraMeshParticles=true
  bUsedWithNiagaraRibbons=true
  bUsedWithNiagaraSprites=true
  blendMode=BLEND_Translucent
  pixelDepthOffsetMode=PDOM_Legacy
  refractionMethod=RM_IndexOfRefraction
  shadingModel=MSM_Unlit
  twoSided=true
  usedShadingModels=Unlit
 }
 nodes [
  Constant() {desc=day time saturation boost, r=-1}
  Constant() {r=0.5}
  n1 = Multiply(A=ComponentMask(None=NamedRerouteUsage()) {desc=Distort mask V2, g=true, r=true}, B=NamedRerouteUsage())
  n2 = MaterialFunctionCall(Float2=TextureCoordinate()) {materialFunction=&BreakOutFloat2Components}
  n3 = MaterialFunctionCall(X=OneMinus(None=@n2.G), Y=@n2.R) {materialFunction=&MakeFloat2}
  n4 = VectorParameter() {group=Flow, parameterName=Flow UV, defaultValue=<rgba 1 1 0 3>}
  n5 = DynamicParameter() {defaultValue=<rgba 0 0 1 0>, paramNames [UVTilingRandom, TimeOffsetRandom, Flicker Time Mult, Dive Speed]}
  n6 = MaterialFunctionCall(X=@n4.R, Y=Add(A=@n5.UVTilingRandom, B=@n4.G)) {materialFunction=&MakeFloat2}
  n7 = MaterialFunctionCall(X=@n4.B, Y=@n4.B) {materialFunction=&MakeFloat2}
  n8 = Add(A=@n5.TimeOffsetRandom, B=Multiply(A=Time(), B=Constant() {r=-1}) {constB=-1})
  n9 = Add(A=Add(A=Multiply(A=@n3.Result, B=@n6.Result), B=Multiply(A=@n7.Result, B=@n8)), B=@n1)
  n10 = TextureSampleParameter2D!(UVs=@n9, Level, Apply View MipBias) {group=Color, mipValueMode=TMVM_MipLevel, parameterName=Color Tex, texture=&T_Tile_RainbowColor}
  Desaturation(None=@n10.RGB, Fraction=ScalarParameter() {group=Color, parameterName=Color Tex Desaturation})
  n11 = ParticleColor()
  n12 = TextureCoordinate()
  n13 = CurveAtlasRowParameter(CurveTime=OneMinus(None=ComponentMask(None=@n12) {r=true})) {
   atlas=/Game/Athena/LinearColorCurves/VFX_CurveAtlas
   curve=/Game/Athena/LinearColorCurves/Curve_Contrail_LoneWolf_Color_Blue
   defaultValue=52
   group=Curve
   parameterName=Curve
  }
  n14 = VectorParameter() {group=Flow, parameterName=Flow Refine, defaultValue=<rgba 1 1 1 0.5>}
  n15 = TextureObjectParameter() {parameterName=Flow Texture, samplerType=SAMPLERTYPE_LinearColor, texture=/Game/Effects/Fort_Effects/Textures/T_WavePattern_Chromatic}
  n16 = MaterialFunctionCall!(Time Mult=ScalarParameter() {defaultValue=0.5, group=Flow, parameterName=FlowFlicker}) {materialFunction=/Game/Effects/Fort_Effects/MaterialFunctions/MF_Flicker}
  n17 = MaterialFunctionCall!(In Texture=@n15, In TexCoords=@n9, Offset Amount=LinearInterpolate(Alpha=Saturate(None=@n16.Raw Flicker)) {constB=0.1}) {materialFunction=/Game/Effects/Fort_Effects/MaterialFunctions/MF_Chroma_UV}
  n18 = Desaturation(None=@n17.Result) {desc=black and white mask}
  n19 = Multiply(A=Power(Base=@n18, Exp=@n14.R), B=@n14.G)
  n20 = Multiply(A=@n18, B=@n14.B)
  n21 = VectorParameter() {group=Flow, parameterName=Flow Color, channelNames {a=Add, b="", g="", r=""}, defaultValue=<rgba 1 1 1 0>}
  n22 = Desaturation(None=@n17.Result, Fraction=ScalarParameter() {defaultValue=-10, group=Flow, parameterName=Flow Desaturation})
  n23 = Multiply(A=Multiply(A=LinearInterpolate(A=@n19, B=@n20, Alpha=@n14.A), B=Add(A=@n22, B=@n21.Add)), B=@n21)
  n24 = TextureObjectParameter() {group=Color, parameterName=Color Tex, texture=&T_Tile_RainbowColor}
  n25 = DynamicParameter() {defaultValue=<rgba 0 0 1 0>, paramNames [UVTilingRandom, TimeOffsetRandom, Flicker Time Mult, Dive Speed]}
  n26 = LinearInterpolate(
   B=ScalarParameter() {defaultValue=0.5, group=Color, parameterName=OffsetAmount}
   Alpha=Saturate(None=MaterialFunctionCall(Time Mult=@n25.Flicker Time Mult) {materialFunction=&MF_Flicker}.Raw Flicker)
  ) {constB=0.5}
  n27 = TextureCoordinate()
  n28 = VectorParameter() {group=Color, parameterName=Color UV, defaultValue=<rgba 1 1 0 0>}
  n29 = MaterialFunctionCall(X=@n28.R, Y=Add(A=@n28.G, B=@n25.UVTilingRandom)) {materialFunction=&MakeFloat2}
  n30 = MaterialFunctionCall(X=@n28.B, Y=@n28.B) {materialFunction=&MakeFloat2}
  n31 = Add(A=@n25.TimeOffsetRandom, B=Time())
  n32 = Add(A=Add(A=Multiply(A=@n27, B=@n29.Result), B=Multiply(A=@n30.Result, B=@n31)), B=@n1)
  n33 = Multiply(
   A=Desaturation(
    None=MaterialFunctionCall(In Texture=@n24, In TexCoords=@n32, Offset Amount=@n26) {materialFunction=&MF_Chroma_UV}.Result
    Fraction=ScalarParameter() {group=Color, parameterName=Color Tex Desaturation}
   )
   B=@n23
  ) {desc=Mix 2 Flow Colors}
  n34 = ScalarParameter() {defaultValue=1, group=Color, parameterName=Use Color Tex}
  n35 = VectorParameter() {group=Flow, parameterName=Inside Color, channelNames {a=SizePower, b="", g="", r=""}, defaultValue=<rgba 1 1 1 1>}
  n36 = OneMinus(None=TextureCoordinate())
  n37 = MaterialFunctionCall(Float2=TextureCoordinate()) {materialFunction=&BreakOutFloat2Components}
  n38 = DynamicParameter() {defaultValue=<rgba 0 0 1 0>, paramNames [UVTilingRandom, TimeOffsetRandom, Flicker Time Mult, Dive Speed]}
  n39 = VectorParameter() {group=Mask, parameterName=Mask UV, defaultValue=<rgba 0.2 1 0.5 0>}
  n40 = Multiply(
   A=MaterialFunctionCall(X=@n37.R, Y=@n37.R) {materialFunction=&MakeFloat2}.Result
   B=MaterialFunctionCall(X=Add(A=@n38.UVTilingRandom, B=@n39.R), Y=@n39.G) {materialFunction=&MakeFloat2}.Result
  )
  n41 = DynamicParameter() {parameterIndex=1, defaultValue=<rgba 1 0 1 -5>, paramNames [Dive Scroll Speed, "", "", ""]}
  n42 = Multiply(
   A=MaterialFunctionCall(X=@n39.B, Y=@n39.B) {materialFunction=&MakeFloat2}.Result
   B=Multiply(A=Time(), B=Multiply(A=@n41.Dive Scroll Speed) {constB=-1}) {constB=0}
  )
  n43 = Multiply(A=@n1, B=Constant2Vector() {g=0.1, r=1})
  n44 = TextureSampleParameter2D!(UVs=Add(A=@n43, B=Add(A=@n40, B=@n42)) {desc=Additional Distortion}, Apply View MipBias) {
   group=Mask
   parameterName=Mask Texture
   samplerType=SAMPLERTYPE_Grayscale
   texture=/Game/Effects/Fort_Effects/Textures/Smoke/T_Smoke_Trail_Soft_NoClamp
  }
  n45 = Multiply(A=Saturate(None=ComponentMask(None=@n36) {r=true}), B=@n44.R) {desc=Feather Scroll Mask}
  n46 = Power(Base=@n45, Exp=@n35.SizePower)
  n47 = LinearInterpolate(A=Add(A=LinearInterpolate(A=@n23, B=@n33, Alpha=@n34)) {constB=0.25}, B=@n35, Alpha=Saturate(None=Power(Base=@n46) {constExponent=3}))
  n48 = Multiply(A=@n47, B=@n13)
  n49 = ScalarParameter() {group=Curve, parameterName=Use Curve}
  n50 = Multiply(
   A=Desaturation(None=LinearInterpolate(A=@n47, B=@n48, Alpha=@n49), Fraction=ScalarParameter() {parameterName=PreParticleColorDesaturation})
   B=@n11.RGB
  ) {desc=Apply Particle Color}
  n51 = Desaturation(None=@n50, Fraction=Constant() {r=-0.25})
  n52 = VectorParameter() {group=ColorTimeOfDay, parameterName=ColorTimeOfDay, channelNames {a=Night, b=Evening, g=Day, r=Morning}, defaultValue=<rgba 1 1 1 1>}
  n53 = Multiply(A=@n51, B=@n52.Morning)
  n54 = MaterialFunctionCall(
   Color_Morning=@n53
   Color_Day=Desaturation(None=@n53, Fraction=ScalarParameter() {defaultValue=-1, group=ColorTimeOfDay, parameterName=DesaturationDay})
   Color_Evening=Multiply(A=@n51, B=@n52.Evening)
   Color_Night=Desaturation(None=Multiply(A=@n52.Night, B=@n51), Fraction=ScalarParameter() {group=ColorTimeOfDay, parameterName=DesaturationNight})
  ) {materialFunction=&Color_Time_of_Day}
  MaterialFunctionCall(Emissive=LinearInterpolate(A=@n51, B=@n54.Result, Alpha=ScalarParameter() {defaultValue=1, parameterName=UseTODColors})) {materialFunction=&MF_FX_EmissiveAdjustments}
  Multiply(A=@n46) {constB=0.5}
  Multiply(A=@n50) {constB=10}
  Multiply(A=@n51, B=@n52.Day)
  Multiply(A=DynamicParameter() {parameterIndex=1, defaultValue=<rgba 1 0 1 -5>, paramNames [Dive Scroll Speed, "", "", ""]}.Dive Scroll Speed) {constB=-1}
  n55 = Time()
  Multiply(A=DynamicParameter() {parameterIndex=1, defaultValue=<rgba 1 0 1 -5>, paramNames [Dive Scroll Speed, "", "", ""]}.Dive Scroll Speed, B=@n55) {constB=-1}
  n56 = MaterialFunctionCall(Float2=TextureCoordinate()) {materialFunction=&BreakOutFloat2Components}
  n57 = MaterialFunctionCall(X=OneMinus(None=@n56.G), Y=@n56.R) {materialFunction=&MakeFloat2}
  n58 = VectorParameter() {group=Noise, parameterName=Noise UV, defaultValue=<rgba 1 0.5 0 -0.2>}
  n59 = DynamicParameter() {defaultValue=<rgba 0 0 1 0>, paramNames [UVTilingRandom, TimeOffsetRandom, Flicker Time Mult, Dive Speed]}
  n60 = MaterialFunctionCall(X=@n58.R, Y=Add(A=@n59.UVTilingRandom, B=@n58.G)) {materialFunction=&MakeFloat2}
  n61 = MaterialFunctionCall(X=@n58.B, Y=@n58.B) {materialFunction=&MakeFloat2}
  n62 = Add(A=@n59.TimeOffsetRandom, B=@n55)
  n63 = TextureSampleParameter2D(UVs=Add(A=Multiply(A=@n57.Result, B=@n60.Result), B=Multiply(A=@n61.Result, B=@n62))) {
   group=Noise
   parameterName=Noise Texture
   samplerType=SAMPLERTYPE_Normal
   texture=&T_Weap_Distort_N_02
  }
  NamedRerouteDeclaration(None=@n63.RGB) {name=Normal_Noise, nodeColor=<rgba 0.8047 1 0 1>}
  n64 = VectorParameter() {parameterName=Base, channelNames {a=NoiseStrength, b=MaskMultiply, g=DepthFade, r=Refraction}, defaultValue=<rgba 0 50 1 0.2>}
  NamedRerouteDeclaration(None=@n64.Refraction) {name=Refraction_Param, nodeColor=<rgba 0.4531 0 1 1>}
  NamedRerouteDeclaration(None=@n64.DepthFade) {name=DepthFade_Param, nodeColor=<rgba 1 0.3516 0 1>}
  NamedRerouteDeclaration(None=@n64.MaskMultiply) {name=MaskMultiply_Param, nodeColor=<rgba 1 0 0.0469 1>}
  NamedRerouteDeclaration(None=@n64.NoiseStrength) {name=NoiseStrength_Param, nodeColor=<rgba 1 0 0.2578 1>}
  n65 = Constant3Vector()
  n66 = NamedRerouteUsage()
  n67 = QualitySwitch(Default=@n66, Low=@n65, High=@n66, Medium=@n65, Epic=@n66)
  Reroute(None=ShadingPathSwitch(Default=@n67, Deferred=@n67, Forward=@n65, Mobile=@n65)) {desc=Normal}
  n68 = OneMinus(None=MaterialFunctionCall(Distance=ScalarParameter() {defaultValue=256, parameterName="Distance (S)"}) {materialFunction=&Near_Camera_Fade}.Result)
  n69 = MaterialFunctionCall(Float2=TextureCoordinate() {uTiling=0.5, vTiling=0.5}) {materialFunction=&BreakOutFloat2Components}
  n70 = Add(
   A=MaterialFunctionCall(X=OneMinus(None=@n69.G), Y=@n69.R) {materialFunction=&MakeFloat2}.Result
   B=Constant2Vector() {g=0.45, r=-0.25}
  )
  n71 = ComponentMask(None=TextureCoordinate()) {r=true}
  n72 = Saturate(None=Add(
    A=TextureSample(UVs=@n70) {samplerSource=SSM_Clamp_WorldGroupSettings, texture=/Game/WaterAndWind/Textures/WeaponFireVectorMask}.R
    B=OneMinus(None=@n71)
   ))
  n73 = Multiply(A=Multiply(A=@n72, B=@n72) {constB=2}) {constB=3}
  n74 = Constant() {r=64}
  n75 = MaterialFunctionCall(Float2=TextureCoordinate()) {materialFunction=&BreakOutFloat2Components}
  n76 = MaterialFunctionCall(X=@n75.G, Y=OneMinus(None=@n75.R)) {materialFunction=&MakeFloat2}
  n77 = Multiply(
   A=TextureSampleParameter2D(UVs=Add(A=@n76.Result, B=@n43)) {parameterName=Base_mask, texture=/Game/Effects/Fort_Effects/Textures/Masks/T_Square_Gradient_M}.R
   B=@n45
  ) {desc=Apply Base Mask}
  n78 = NamedRerouteUsage()
  n79 = Multiply(
   A=Multiply(A=@n11.A, B=MaterialFunctionCall(Distance=@n74) {materialFunction=&Near_Camera_Fade}.Result)
   B=Saturate(None=Multiply(A=@n77, B=@n78))
  ) {desc=Apply Combined Mask}
  n80 = LinearInterpolate(A=@n79, B=Multiply(A=@n79, B=Saturate(None=@n73)), Alpha=ScalarParameter() {parameterName=ButtomFade})
  n81 = DepthFade(Opacity=@n80, FadeDistance=NamedRerouteUsage())
  n82 = ShadingPathSwitch(Default=@n81, Deferred=@n81, Forward=@n80, Mobile=@n80)
  n83 = LinearInterpolate(Alpha=Multiply(A=Multiply(A=@n82, B=@n68), B=NamedRerouteUsage())) {constA=1, constB=1.05}
  n84 = Constant() {r=1}
  n85 = QualitySwitch(Default=@n83, Low=@n84, High=@n83, Medium=@n84, Epic=@n83)
  Reroute(None=ShadingPathSwitch(Default=@n85, Deferred=@n85, Forward=@n84, Mobile=@n84)) {desc=Refraction}
  Reroute(None=@n54.Result) {desc=Emissive Color}
  n86 = Fresnel(ExponentIn=ScalarParameter() {defaultValue=0.2, parameterName=FresnelExp})
  n87 = LinearInterpolate(A=@n82, B=Multiply(A=@n82, B=OneMinus(None=@n86)), Alpha=ScalarParameter() {parameterName=UseFresnel})
  Reroute(None=ShadingPathSwitch(Default=@n87, Deferred=@n87, Forward=@n82, Mobile=@n82)) {desc=Opacity}
  TextureObject() {texture=/Game/TimeOfDay/TODM/BR/s11/IWD_skybox/T_Noise_Cloud_IWD}
  TextureObject() {samplerType=SAMPLERTYPE_Normal, texture=/Game/Global/Textures/Default/Flowmaps/T_SwirlFlow_02_N}
 ]
}
